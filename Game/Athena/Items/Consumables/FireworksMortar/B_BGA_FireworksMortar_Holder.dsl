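; Fireworks mortar holder actor that handles replicated orientation, movement resumption and destruction.
; The rotation RepNotify yaws the actor from land rotation plus an offset, pitches the top by the arrow placement pitch, and angles the base pivot from the hit normal.
; Other RepNotifies reattach the projectile movement component to the capsule when simulation resumes, and call DestroyHolder once destruction starts.
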(fn OnRep_ApplyRotation
  ;[1 1 1] Set Actor Rotation
  (Transformation|SetActorRotation (Math|Rotator|MakeRotator 0.0 0.0 (+ (.yaw (Variables|HandleRotation|GetMeshLandRotation)) (Variables|HandleRotation|GetYawOffset))))
  ;[1 1 1] Calculate the intended rotation angle of the stand using the arrow placement pitch and if chosen the ground offset.
  (Transformation|SetRelativeRotation (Variables|Default|GetTop) (Math|Rotator|MakeRotator (Variables|HandleRotation|GetArrowPlacementPitch) 0.0 9e+01) false false)
  ;[1 1 1] Angle the stand to mount to the Ground
  (Transformation|SetRelativeRotation (Variables|Default|GetBasePivot) (Math|Rotator|MakeRotator 0.0 (.pitch (Math|Rotator|MakeRotfromZX (Variables|HitEvent|GetHitNormal) (Transformation|GetForwardVector (Variables|Default|GetBasePivot))))) false false))

(fn OnRep_bResumeSimulation
  (if (Variables|HitEvent|GetbResumeSimulation)
    (Components|Movement|SetUpdatedComponent (Variables|Default|GetProjectileMovement) (Variables|Default|GetCapsule))))

(fn OnRep_StartedDestroying
  (if (Variables|Death|GetStartedDestroying)
    (CallFunction|DestroyHolder)))

(fn ConstructionScript)
