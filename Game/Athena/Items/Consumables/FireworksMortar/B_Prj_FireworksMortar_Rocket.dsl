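; Fireworks Mortar rocket projectile that explodes on bounce, on overlapping water (SurfaceType8), or after a random lifetime.
; Construction randomizes hue and mesh scale.
; On explosion, a cosmetic-only branch plays sound, surface-specific FX and a sound indicator, then removes the mesh.
; That same cosmetic branch rolls a scalable-float chance to spawn a fireball projectile along a randomly tilted and rotated up vector.
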(event EventOnBounce
  (params (Hit :type Hit%20Result%20Structure%20%28by%20ref%29))
  (bind (_bblockinghit _binitialoverlap _time _distance _location _impactpoint _normal) (Collision|BreakHitResult Hit))
  ;[1 1 1] If bounce off an enemy explode, if bounce off something else small delay then explode
  (Variables|Default|SetHitResult Hit)
  (Variables|Default|SetHitResultNormal _normal)
  (Game|Components|ProjectileMovement|SetVelocityinLocalSpace (Variables|Movement|GetProjectileMovementComponent) "0, 0, 0")
  (Delivery|DoExplosion))

(event EventBeginPlay
  ;[1 1 1] On Begin Play sound, set collision enabled and lifetime and wait to do explosion
  (Audio|SpawnSoundAttached (Variables|SoundAndVisuals|GetSpawnSound) (Variables|Collision|GetCapsuleComponent)) ; (Replace with GC when replecated GCs become available)
  (Collision|SetActorEnableCollision true)
  (Utilities|FlowControl|Delay (Math|Random|RandomFloatinRange (Variables|ProjVars|Lifetime|GetLifetime_Min) (Variables|ProjVars|Lifetime|GetLifetime_Max)))
  (Delivery|DoExplosion))

(event EventOnExploded
  (params (HitActors :type Array%20of%20Actor%20Object%20References) (HitResults :type Array%20of%20Hit%20Result%20Structures))
  (bind _hitresultnormal (Variables|Default|GetHitResultNormal))
  (bind _returnvalue (Math|Conversions|RotationFromXVector _hitresultnormal))
  (bind _returnvalue_2 (Transformation|GetActorLocation))
  (bind _mesh (Variables|ProjectileFirecrackerMini|GetMesh))
  ;[1 1 1] Play Explode Sound and FX (Replace with GC when replecated GCs become available)
  (Networking|CanExecuteCosmeticEvents
    (:True
      (Audio|PlaySoundatLocation (Variables|SoundandVisuals|GetExplosionSound) _returnvalue_2)
      (if (Utilities|Enum|Equal(Enum) (Variables|Default|GetSurface%20Type) "SurfaceType8")
        (Effects|SpawnEmitteratLocation (Variables|SoundandVisuals|GetP_Explosion_Water) _returnvalue_2 _returnvalue)
        (else
          (Effects|SpawnEmitteratLocation (Variables|SoundandVisuals|GetP_Explosion) _returnvalue_2 _returnvalue)))
      (Audio|Components|Audio|FadeOut (Variables|Default|GetAudio_Proj_Loop) 0.1 0.0 "Linear" :!class /Script/Engine.AudioComponent)
      ;;[1 1 1] Sound Indicator For Phones
      (Fortnite|ShowSoundIndicatorLocalforTeam :Location (Transformation|GetActorLocation) :MaxAudibleDistance (Variables|/Script/Engine.SoundBase|GetMaxDistance (Variables|SoundandVisuals|GetExplosionSound)) :IndicatorType "Gunshot" :Team (Team|GetActorTeam self) :Affiliations (Utilities|Array|MakeArray "Hostile" "Neutral") :Tint "(R=0.000000,G=0.000000,B=0.000000,A=0.000000)" :AttenuationMaxRangeMultiplier 1.0 :AttenuationIntensityMultiplier 1.0)
      ;[1 1 1] ScaleDown Mesh and DeactivateFX
      (Transformation|SetWorldScale3D _mesh "0.010000,0.010000,0.010000")
      (Components|Activation|Deactivate (Variables|ProjectileFirecrackerMini|GetP_Trail))
      (Components|DestroyComponent _mesh)
      ;[1 1 1] Chance to spawn fireball
      (CallFunction|TrySpawnFireball))))

(event EventOnTouched
  (params (OtherActor :type Actor%20Object%20Reference) (OtherComp :type Primitive%20Component%20Object%20Reference) (HitResult :type Hit%20Result%20Structure%20%28by%20ref%29) (bIsOverlap :type Boolean))
  (bind (_bblockinghit _binitialoverlap _time _distance _location _impactpoint _normal _impactnormal _physmat) (Collision|BreakHitResult HitResult))
  (if bIsOverlap
    (Variables|Default|SetSurface%20Type (select (Utilities|IsValid _physmat :!class /Script/Engine.KismetSystemLibrary) (Variables|/Script/PhysicsCore.PhysicalMaterial|GetSurfaceType _physmat) "SurfaceType_Default"))
    (if (Utilities|Enum|Equal(Enum) (Variables|Default|GetSurface%20Type) "SurfaceType8")
      (Delivery|DoExplosion))))

(event Custom|TrySpawnFireball
  ;[1 1 1] Check if a fireball should be spawned using random chance, then spawn a fireball launched in a random direction if so
  (bind _bshouldspawn (CallFunction|ShouldSpawnFireball))
  (if _bshouldspawn
    (bind _direction (CallFunction|GetRandomFireballDirection))
    (CallFunction|SpawnFireball _direction)))

(fn GetRandomFireballDirection
  (return (Math|Vector|RotateVectorAroundAxis (Math|Vector|RotateVectorAroundAxis "0, 0,1.000000" (Math|Random|RandomFloatinRange (Variables|Default|GetMinFireballTilt) (Variables|Default|GetMaxFireballTilt)) "1.000000, 0, 0") (Math|Random|RandomFloatinRange 0.0 359.0) "0.000000,0.000000,1.000000")))

(fn ShouldSpawnFireball
  ;[1 1 1] Grabs a random value between 0 and 1 and compares it to the chance of spawning a fireball
  (return (Math|Random|RandomBoolwithWeight (ScalableFloat|GetValueatLevel (Variables|Default|GetRow_FireballSpawnChance)))))

(fn SpawnFireball
  (params (Direction :type Vector))
  (bind (_fireballspeedrange_x _fireballspeedrange_y) (Math|Vector2D|BreakVector2D (Variables|GetFireballSpeedRange_X)))
  (Projectile|SpawnProjectile :ProjectileClass "/Game/Athena/Items/Consumables/FireworksMortar/B_Prj_Athena_FireworksMortar_Fireball.B_Prj_Athena_FireworksMortar_Fireball_C" :RequestedBy self :SpawnLocation (Transformation|GetActorLocation) :SpawnRotation "0, 0, 0" :SpawnDirection (Math|Rotator|FindLookatRotation (Transformation|GetActorLocation) (+ (Transformation|GetActorLocation) Direction)) :bUseDefaultPhysics false :InitialSpeed (select (Variables|Fireball|GetbUseRangeForFireballSpeed) (Math|Random|RandomFloatinRange _fireballspeedrange_x _fireballspeedrange_y) _fireballspeedrange_x) :GravityScale 1.0 :ChargePercent 0.0))

(fn ConstructionScript
  (bind _mesh (Variables|ProjectileFirecrackerMini|GetMesh))
  (bind _returnvalue (Rendering|Material|CreateDynamicMaterialInstance _mesh 0 0 "None" :!class /Script/Engine.PrimitiveComponent))
  (bind _output_get (Variables|SoundandVisuals|SetDynamic_Material _returnvalue))
  (Rendering|Material|SetMaterial _mesh 0 _output_get)
  ;[0 0 0] Randomized the color and scale for the rocket
  (Rendering|Material|SetScalarParameterValue _output_get "HueShiftColorRange" (Utilities|Array|Get(aref) (Variables|Default|GetHueShiftColorRange) (Math|Random|RandomInteger 5)) :!class /Script/Engine.MaterialInstanceDynamic)
  (Transformation|SetWorldScale3D (Variables|ProjectileFirecrackerMini|GetMesh) (* (Transformation|GetWorldScale (Variables|ProjectileFirecrackerMini|GetMesh)) (Math|Vector|MakeVector (Math|Random|RandomFloatinRange 0.5 1.0) (Math|Random|RandomFloatinRange 0.5 1.0) (Math|Float|Lerp 0.5 1.3 (Math|Random|RandomFloatinRange 0.5 1.0))))))
