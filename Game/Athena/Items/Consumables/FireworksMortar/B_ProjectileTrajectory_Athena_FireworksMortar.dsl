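; Trajectory preview actor for the Athena Fireworks Mortar consumable, extending a parent trajectory class.
; On BeginPlay it hides the Target component and resizes the spline mesh component array to the max mesh count.
; Each trajectory spline update runs the parent logic, then sets the Target's world yaw from the instigator pawn's targeting rotation plus yaw and mesh offsets, keeping its pitch and roll.
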(event EventBeginPlay
  (Rendering|SetVisibility (Variables|Default|GetTarget) false false)
  (Utilities|Array|Resize (Variables|Trajectory|GetSplineMeshComponents) (Variables|Default|GetMaxMeshCount)))

(event Collision|EventActorBeginOverlap
  (params (OtherActor :type Actor%20Object%20Reference))
  (|CallParentFunction OtherActor) ; This node is disabled and will not be called.\nDrag off pins to build functionality.
)

(event EventTick
  (params (DeltaSeconds :type Float%20%28single-precision%29))
  (|CallParentFunction DeltaSeconds) ; This node is disabled and will not be called.\nDrag off pins to build functionality.
)

(event Trajectory|EventSetTrajectorySpline
  (params (SplinePoints :type Array%20of%20Vectors) (SplineTangents :type Array%20of%20Vectors))
  ;[1 1 1] When the Trajectory updates also update the rotation of the Target Mesh
  (|CallParentFunction SplinePoints SplineTangents)
  (bind _asfort_player_pawn (Utilities|Casting|CastToFortPlayerPawn (Game|GetInstigator)))
  (Transformation|SetWorldRotation (Variables|Default|GetTarget) (Math|Rotator|MakeRotator (.roll (Transformation|GetWorldRotation (Variables|Default|GetTarget))) (.pitch (Transformation|GetWorldRotation (Variables|Default|GetTarget))) (Math|Float|float+float (.yaw (Targeting|GetDefaultTargetingRotation _asfort_player_pawn)) (Variables|Default|GetYawOffset) (Variables|Default|GetMeshRotationOffset))) false false))

(fn ConstructionScript
  (|CallParentFunction))
