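; Event-triggered gameplay ability that launches a volley of bottle rockets from a Fireworks Mortar holder passed in the event data.
; On activation it optionally loads scalable-float balance values and builds two explosion effect container specs for the projectiles.
; A randomly delayed loop spawns rockets along the holder's up vector with random spread and speed, boosting some for higher shots.
; After max shots it flags the holder as destroying and ends.
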(event Ability|EventActivateAbilityFromEvent
  (params (EventData :type Gameplay%20Event%20Data%20Structure%20%28by%20ref%29))
  (bind (_eventtag _instigator _target _optionalobject) (Utilities|Struct|BreakGameplayEventData EventData))
  (|CallParentFunction EventData)
  ;[1 1 1] Set Balance using sheets.
  (if (Variables|GAVars|Rows|GetUseBalanceRows)
    (CallFunction|SetBalance)
    (else
      _))
  ;[1 1 1] Save reference to rocket
  (bind _asb_bga_fireworks_mortar_holder (Utilities|Casting|CastToB_BGA_FireworksMortar_Holder _optionalobject))
  (Variables|GAVars|SetRocketHolder _asb_bga_fireworks_mortar_holder)
  ;[1 1 1] Save out spec for later use
  (Variables|GAVars|Damage|SetGECS_Explode (Ability|MakeGameplayEffectContainerSpec (Variables|GAVars|Damage|GetT_Targeting)))
  (Variables|GAVars|Damage|SetGECS_EXplode_additional (Ability|MakeGameplayEffectContainerSpec (Variables|GAVars|Damage|GetEnvTargeting)))
  ;[1 1 1] Start Firing Rockets
  (CallFunction|FireRockets))

(event Custom|FireRockets
  ;[1 1 1 0.5] Loop through and fire rockets
  (Variables|GAVars|RocketSpawn|SetShotsFired 0.0)
  (CallFunction|SpawnRocket)
  (Class|BBGAFireworksMortarHolder|MuzzleFlash (Variables|GAVars|GetRocketHolder))
  ;;[1 1 1] Increment and check shots fired
  (bind _result (Math|Float|IncrementFloat (Variables|GAVars|RocketSpawn|GetShotsFired))
    (:%20%20
      (if (not (>= _result (Variables|GAVars|RocketSpawn|GetMaxShots)))
        ;;[1 1 1] Create random delayed loop
        (Utilities|FlowControl|Delay (CallFunction|AddVariance (Variables|GAVars|RocketSpawn|GetSpawnDelay) (Variables|GAVars|RocketSpawn|GetSpawnDelayVarience))))
      ;;[1 1 1] End ability after max shots reached and tell holder to destroy
      (Development|PrintString "Hit Max Shots")
      (Variables|/Game/Athena/Items/Consumables/FireworksMortar/B_BGA_FireworksMortar_Holder.B_BGA_FireworksMortar_Holder_C|SetStartedDestroying true (Variables|GAVars|GetRocketHolder))
      (Ability|EndAbility))))

(fn AddVariance
  (params (Default :type Float%20%28double-precision%29) (Variance :type Float%20%28double-precision%29))
  (return (Math|Random|RandomFloatinRange (- Default Variance) (+ Default Variance))))

(fn SetBalance
  (Variables|GAVars|RocketSpawn|SetBottleRocketSpeed (ScalableFloat|GetValueatLevel (Variables|GAVars|Rows|GetRow_RocketSpeed)))
  (Variables|GAVars|RocketSpawn|SetBottleRocketSpeedVariance (ScalableFloat|GetValueatLevel (Variables|GAVars|Rows|GetRow_RocketSpeedVariance)))
  (Variables|GAVars|RocketSpawn|SetSpawnDelay (ScalableFloat|GetValueatLevel (Variables|GAVars|Rows|GetRow_SpawnDelay)))
  (Variables|GAVars|RocketSpawn|SetSpawnDelayVarience (ScalableFloat|GetValueatLevel (Variables|GAVars|Rows|GetRow_SpawnDelayVariance)))
  (Variables|GAVars|RocketSpawn|SetMaxShots (ScalableFloat|GetValueatLevel (Variables|GAVars|Rows|GetRow_MaxShots)))
  (Variables|GAVars|RocketSpawn|SetSpread (ScalableFloat|GetValueatLevel (Variables|GAVars|Rows|GetRow_Spread))))

(fn SpawnRocket
  ;[0.841 0.9162 1] % of rockets will be shot vertically higher (for flavor)
  (Variables|GAVars|RocketSpawn|SetHighShot (Math|Random|RandomBoolwithWeight (ScalableFloat|GetValueatLevel (Variables|GAVars|Rows|GetRow_HighShot))))
  ;[1 1 1] Spawn projectile on server
  (bind _returnvalue (Projectile|SpawnProjectile :ProjectileClass (Variables|GAVars|RocketSpawn|GetPrj_BottleRocket) :RequestedBy (Ability|GetAvatarActorfromActorInfo) :SpawnLocation (+ (Transformation|GetWorldLocation (Variables|/Game/Athena/Items/Consumables/FireworksMortar/B_BGA_FireworksMortar_Holder.B_BGA_FireworksMortar_Holder_C|GetTop (Variables|GAVars|GetRocketHolder))) (* (Math|Vector|Normalize (Transformation|GetUpVector (Variables|/Game/Athena/Items/Consumables/FireworksMortar/B_BGA_FireworksMortar_Holder.B_BGA_FireworksMortar_Holder_C|GetTop (Variables|GAVars|GetRocketHolder)))) (Variables|GAVars|RocketSpawn|GetSpawnLocOffset))) :SpawnRotation (Math|Conversions|RotationFromXVector (Math|Vector|MakeVector (+ (.x (Math|Vector|GetUpVector (Transformation|GetWorldRotation (Variables|/Game/Athena/Items/Consumables/FireworksMortar/B_BGA_FireworksMortar_Holder.B_BGA_FireworksMortar_Holder_C|GetTop (Variables|GAVars|GetRocketHolder))))) (Math|Random|RandomFloatinRange (* (Variables|GAVars|RocketSpawn|GetSpread) -1) (Variables|GAVars|RocketSpawn|GetSpread))) (+ (.y (Math|Vector|GetUpVector (Transformation|GetWorldRotation (Variables|/Game/Athena/Items/Consumables/FireworksMortar/B_BGA_FireworksMortar_Holder.B_BGA_FireworksMortar_Holder_C|GetTop (Variables|GAVars|GetRocketHolder))))) (Math|Random|RandomFloatinRange (* (Variables|GAVars|RocketSpawn|GetSpread) -1) (Variables|GAVars|RocketSpawn|GetSpread))) (+ (.z (Math|Vector|GetUpVector (Transformation|GetWorldRotation (Variables|/Game/Athena/Items/Consumables/FireworksMortar/B_BGA_FireworksMortar_Holder.B_BGA_FireworksMortar_Holder_C|GetTop (Variables|GAVars|GetRocketHolder))))) (Math|Random|RandomFloatinRange (* (Variables|GAVars|RocketSpawn|GetSpread) -1) (Variables|GAVars|RocketSpawn|GetSpread))))) :SpawnDirection (Math|Conversions|RotationFromXVector (Math|Vector|MakeVector (+ (.x (Math|Vector|GetUpVector (Transformation|GetWorldRotation (Variables|/Game/Athena/Items/Consumables/FireworksMortar/B_BGA_FireworksMortar_Holder.B_BGA_FireworksMortar_Holder_C|GetTop (Variables|GAVars|GetRocketHolder))))) (Math|Random|RandomFloatinRange (* (Variables|GAVars|RocketSpawn|GetSpread) -1) (Variables|GAVars|RocketSpawn|GetSpread))) (+ (.y (Math|Vector|GetUpVector (Transformation|GetWorldRotation (Variables|/Game/Athena/Items/Consumables/FireworksMortar/B_BGA_FireworksMortar_Holder.B_BGA_FireworksMortar_Holder_C|GetTop (Variables|GAVars|GetRocketHolder))))) (Math|Random|RandomFloatinRange (* (Variables|GAVars|RocketSpawn|GetSpread) -1) (Variables|GAVars|RocketSpawn|GetSpread))) (+ (.z (Math|Vector|GetUpVector (Transformation|GetWorldRotation (Variables|/Game/Athena/Items/Consumables/FireworksMortar/B_BGA_FireworksMortar_Holder.B_BGA_FireworksMortar_Holder_C|GetTop (Variables|GAVars|GetRocketHolder))))) (Math|Random|RandomFloatinRange (* (Variables|GAVars|RocketSpawn|GetSpread) -1) (Variables|GAVars|RocketSpawn|GetSpread))))) :EffectContainerSpecToApplyOnExplode (Variables|GAVars|Damage|GetGECS_Explode) :bUseDefaultPhysics false :InitialSpeed (select (Variables|GAVars|RocketSpawn|GetHighShot) (* (CallFunction|AddVariance (Variables|GAVars|RocketSpawn|GetBottleRocketSpeed) (Variables|GAVars|RocketSpawn|GetBottleRocketSpeedVariance)) (Math|Random|RandomFloatinRange 1.0 1.3)) (CallFunction|AddVariance (Variables|GAVars|RocketSpawn|GetBottleRocketSpeed) (Variables|GAVars|RocketSpawn|GetBottleRocketSpeedVariance))) :GravityScale (LocalVariables|GetBottleRocketGravity) :ChargePercent 0.0))
  (Delivery|AddAdditionalExplodeEffectContainer _returnvalue (Variables|GAVars|Damage|GetGECS_EXplode_additional)))
