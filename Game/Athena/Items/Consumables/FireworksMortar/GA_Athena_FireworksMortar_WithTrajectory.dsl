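; Gameplay ability for the fireworks mortar consumable that adds a trajectory preview and launches the mortar holder projectile with a matching yaw.
; On activation it reads a yaw offset from a scalable float row and applies a tracker effect to the owner.
; Trajectory setup and cleanup send gameplay events to switch the holder indicator, passing that offset to it and the trajectory actor.
; Spawned projectiles receive the player's targeting rotation plus the same offset.
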(event Ability|EventActivateAbility
  ;[1 1 1] Set BalanceValues on start
  (CallFunction|SetBalanceValues)
  (Ability|ApplyGameplayEffectToOwner (Variables|Default|GetGE_Tracker))
  (|CallParentFunction))

(event Ability|EventOnEndAbility
  (params (bWasCancelled :type Boolean))
  (|CallParentFunction bWasCancelled) ; This node is disabled and will not be called.\nDrag off pins to build functionality.
)

(event ProjectileLifecycle|EventOnProjectileSpawned
  (params (NewSpawnedProjectile :type Fort%20Projectile%20Base%20Object%20Reference))
  ;[1 1 1] Pass the Rotation of the player to the projectile
  (bind _asb_prj_fireworks_mortar_holder (Utilities|Casting|CastToB_Prj_FireworksMortar_Holder NewSpawnedProjectile))
  (Class|BPrjFireworksMortarHolder|SetRotation _asb_prj_fireworks_mortar_holder (Targeting|GetDefaultTargetingRotation (Variables|ReferencedActors|GetPlayerPawn)) (Variables|Default|GetYawOffsetMaster)))

(event Trajectory|EventSetupDummyProjectile
  ;[1 1 1] Activate Holder Indicator Ability
  (|CallParentFunction)
  (Ability|SendGameplayEvent "(TagName=\"Event.Athena.FireworksMortarHolder.Trajectory.Activate\")" (Utilities|Struct|MakeGameplayEventData :OptionalObject (Variables|Trajectory|GetTrajectoryIndicator) :ContextHandle "()" :InstigatorTags "(GameplayTags=)" :TargetTags "(GameplayTags=)" :EventMagnitude (Variables|Default|GetYawOffsetMaster) :TargetData "()"))
  ;[1 1 1] Set Yaw Offset for Trejectory Indicator
  (bind _asb_projectile_trajectory_athena_fireworks_mortar (Utilities|Casting|CastToB_ProjectileTrajectory_Athena_FireworksMortar (Variables|Trajectory|GetTrajectoryIndicator)))
  (Variables|/Game/Athena/Items/Consumables/FireworksMortar/B_ProjectileTrajectory_Athena_FireworksMortar.B_ProjectileTrajectory_Athena_FireworksMortar_C|SetYawOffset (Variables|Default|GetYawOffsetMaster) _asb_projectile_trajectory_athena_fireworks_mortar))

(event Trajectory|EventCleanupTrajectory
  ;[1 1 1] Activate Holder Indicator Ability
  (|CallParentFunction)
  (Ability|SendGameplayEvent "(TagName=\"Event.Athena.FireworksMortarHolder.Trajectory.Deactivate\")"))

(fn SetBalanceValues
  (Variables|Default|SetYawOffsetMaster (ScalableFloat|GetValueatLevel (Variables|Default|GetRow_YawOffsetValue))))
