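; Gameplay ability that caps how many Fireworks Mortar holders exist at once, keeping the tracked count to two.
; On activation from a gameplay event, it registers the event's target in an actor tracking group, then stores that group's tracked actors.
; If more than two are tracked, it casts the first entry to the mortar holder class and destroys it, then ends the ability.
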(event Ability|EventActivateAbilityFromEvent
  (params (EventData :type Gameplay%20Event%20Data%20Structure%20%28by%20ref%29))
  (bind (_eventtag _instigator _target) (Utilities|Struct|BreakGameplayEventData EventData))
  (bind _t_trackinggroup (Variables|Default|GetT_TrackingGroup))
  ;[1 1 1] Track fireworks removing them if there are more than 2.
  (ActorTracking|TrackGroupedActor _target _t_trackinggroup)
  (bind _returnvalue (ActorTracking|GetTrackedGroupedActors _t_trackinggroup))
  (bind _output_get (Variables|Default|SetTrackedFireworkHolders _returnvalue))
  (if (> (Utilities|Array|Length _output_get) 2)
    (bind _asb_bga_fireworks_mortar_holder (Utilities|Casting|CastToB_BGA_FireworksMortar_Holder (Utilities|Array|Get(acopy) _output_get))
      (:then
        (Class|BBGAFireworksMortarHolder|DestroyHolder _asb_bga_fireworks_mortar_holder))
      (:CastFailed
        _))
    (Ability|EndAbility)
    (else
      (Ability|EndAbility))))
