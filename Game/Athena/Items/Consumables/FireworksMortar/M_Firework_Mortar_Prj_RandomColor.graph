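# Masked, lit material for the firework mortar projectile that varies its color by hue-shifting the fireworks box diffuse texture.
# A particle Dynamic Parameter, HueShiftColorRange, drives the HueShift function, letting the emitter set the color.
# An offset, clamped bubble texture serves as a stencil that keeps original colors in one region.
# A second branch dims the shifted color and masks it by inverted diffuse blue, alongside normal and mask texture samples.
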Material /Game/Athena/Items/Consumables/FireworksMortar/M_Firework_Mortar_Prj_RandomColor
graph {
 properties {blendMode=BLEND_Masked, ditherOpacityMask=true, pixelDepthOffsetMode=PDOM_Legacy, usedShadingModels=Default Lit}
 nodes [
  n1 = TextureSample() {texture=/Game/Missions/Secondary/Events/HonkingFirework/Props/Textures/T_box_of_fireworks_D}
  n2 = MaterialFunctionCall(Hue Shift Percentage=DynamicParameter() {paramNames [HueShiftColorRange, Param2, Param3, Param4]}.HueShiftColorRange, Texture=@n1.RGB) {materialFunction=&HueShift}
  n3 = MaterialFunctionCall(Float2=TextureCoordinate() {uTiling=0.95, vTiling=3.5}) {materialFunction=&BreakOutFloat2Components}
  n4 = Add(A=@n3.R) {constB=-0.2}
  n5 = Add(A=@n3.G) {constB=-0.45}
  n6 = TextureSample(UVs=Panner(Coordinate=AppendVector(A=@n4, B=@n5), Time=Constant() {r=2.3}) {speedX=0.1}) {samplerSource=SSM_Clamp_WorldGroupSettings, texture=/Game/Effects/Fort_Effects/Textures/Husk/Poison/T_FakeBubble_Spec}
  Add(A=Multiply(A=@n2.Result, B=OneMinus(None=@n6.R)), B=Multiply(A=@n1.RGB, B=@n6.R))
  n7 = LinearInterpolate(Alpha=OneMinus(None=@n1.B)) {constA=-10}
  Divide(A=Multiply(A=@n2.Result, B=Saturate(None=@n7))) {constB=3}
  MaterialFunctionCall() {materialFunction=&DebugTimeSine}
  ParticleColor()
  ScalarParameter() {parameterName=HueShiftColorRange}
  TextureSample() {samplerType=SAMPLERTYPE_Normal, texture=/Game/Missions/Secondary/Events/HonkingFirework/Props/Textures/T_box_of_fireworks_N}
  TextureSample() {samplerType=SAMPLERTYPE_Masks, texture=/Game/Missions/Secondary/Events/HonkingFirework/Props/Textures/T_box_of_fireworks_S}
 ]
}
